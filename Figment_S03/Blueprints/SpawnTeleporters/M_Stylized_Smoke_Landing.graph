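# Translucent, unlit, two-sided particle material for stylized landing smoke, filed with the spawn teleporter assets.
# A depth-of-field far mask blends sharp and pre-blurred smoke textures for color and alpha.
# Color adds a slowly panning, contrast-boosted noise texture, then multiplies by particle color.
# Opacity combines particle alpha with a slight view-angle falloff against world up, then softens through depth fade and near-camera fade.
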Material /Figment_S03/Blueprints/SpawnTeleporters/M_Stylized_Smoke_Landing
graph {
 properties {
  bUsedWithMeshParticles=true
  bUsedWithParticleSprites=true
  blendMode=BLEND_Translucent
  pixelDepthOffsetMode=PDOM_Legacy
  shadingModel=MSM_Unlit
  twoSided=true
  usedShadingModels=Unlit
 }
 nodes [
  Constant() {r=0.5}
  Constant() {r=1}
  n1 = Constant() {r=0.6}
  n2 = Power(
   Base=TextureSample(UVs=Panner() {speedX=0.01, speedY=0.02}) {texture=/Game/Global/Textures/Default/Tileable_Patterns/T_Smoke_OUT}.RGB
   Exp=Constant() {r=2}
  )
  n3 = TextureSample() {texture=/Figment_S03/Effects/Fort_Effects/Textures/T_Stylized_Smoke1}
  n4 = TextureSample() {texture=/Game/Effects/Fort_Effects/Textures/T_Stylized_Smoke1_Blur}
  n5 = DepthOfFieldFunction!(Depth) {functionValue=TDOF_FarMask}
  n6 = ParticleColor()
  Multiply(A=Add(A=Multiply(A=@n1, B=@n2), B=LinearInterpolate(A=@n3.RGB, B=@n4.RGB, Alpha=@n5)), B=@n6.RGB)
  n7 = DotProduct(A=CameraVectorWS(), B=Constant3Vector() {constant=<rgba 0 0 1 0>})
  n8 = Constant() {r=3}
  n9 = Multiply(
   A=Multiply(A=LinearInterpolate(A=@n3.A, B=@n4.A, Alpha=@n5), B=@n6.A)
   B=Clamp(None=Power(Base=@n7, Exp=@n8), Min=Constant() {r=0.8}, Max=Constant() {r=1})
  )
  n10 = Constant()
  n11 = Constant() {r=1}
  Multiply(
   A=DepthFade(Opacity=Clamp(None=@n9, Min=@n10, Max=@n11))
   B=MaterialFunctionCall(Fade Length=Constant() {r=300}, Fade Offset=Constant() {r=50}) {materialFunction=&CameraDepthFade}.Result
  )
 ]
}
